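; Guided motion movement mode asset for the MotoWrestle air-to-ground dash, under the Slide ability, that overrides or supplies the constant-force direction.
; GetDirection_ConstantForce reads the owner pawn's camera targeting rotation and clamps its pitch between hotfixable min and max scalable floats.
; It keeps yaw and roll, then returns the normalized forward vector, so the dash follows camera aim within a limited vertical angle.
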(fn GetDirection_ConstantForce
  (params (GuidedMotionRuntimeData :type Fort%20Movement%20Mode%20Guided%20Motion%20Runtime%20Data%20Object%20Reference))
  (bind _returnvalue (TargetingSystem|GetPlayerCameraTargetingTransform (Variables|FortMovementMode_BaseExtLogic|GetOwnerPawn)))
  (return (Math|Vector|Normalize (Math|Vector|GetForwardVector (Math|Rotator|MakeRotator (.roll (.rotation _returnvalue)) (Math|Float|ClampAngle (.pitch (.rotation _returnvalue)) (ScalableFloat|GetValueatLevel (Variables|Default|GetHotfix_AirToGroundDirectionPitchMin)) (ScalableFloat|GetValueatLevel (Variables|Default|GetHotfix_AirToGroundDirectionPitchMax))) (.yaw (.rotation _returnvalue)))))))
