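; Guided motion movement mode for the MotoWrestle slide ability's dash toward an actor.
; Its visible logic computes the motion's maximum duration from the distance between the player's initial location and the move-to end location in the runtime data.
; That distance is safe-divided by a scalable-float max speed from the move-to-actor guide config, so the time limit scales with dash length.
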(fn GetMaxDuration
  (params (GuidedMotionRuntimeData :type Fort%20Movement%20Mode%20Guided%20Motion%20Runtime%20Data%20Object%20Reference))
  (return (Math|Float|SafeDivide (Math|Vector|Distance(Vector) (Variables|/Script/FortniteGame.FortMovementMode_GuidedMotionRuntimeData|GetLocalMoveToEndLocation GuidedMotionRuntimeData) (Variables|/Script/FortniteGame.FortMovementMode_GuidedMotionRuntimeData|GetLocalInitialPlayerLocation GuidedMotionRuntimeData)) (ScalableFloat|GetValueatLevel (Variables|Config-Guide-MoveToActor|GetMaxSpeed)))))
